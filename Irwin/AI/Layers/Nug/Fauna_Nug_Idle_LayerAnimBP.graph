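# Animation layer Blueprint for the Nug fauna bird that overrides its idle pose with occasional ground-pecking breaks.
# On initialize it caches the main AnimBP_Fauna_Nug, then each update sets ShouldPlayIdleBreak when the Nug has no valid aim target and is not held.
# That flag switches Fauna Idle, using inertialized blends, between a random idle/peck player and a plain idle loop.
# The IdleTurn layer passes its input through unchanged.
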AnimBlueprint /Irwin/AI/Layers/Nug/Fauna_Nug_Idle_LayerAnimBP
target_skeleton=/Irwin/AI/Base/Nug_Bird_Invis_Skeleton
graphs [
 AnimGraph {nodes [Output Pose() {layerGroup=None}]}
 EventGraph {nodes [
   Event BlueprintUpdateAnimation()
   n1 = Get NugAnimBP() -> (NugAnimBP:Anim BP Fauna Nug Object Reference)
   > Is Valid(InputObject=@n1)
   n2 = Get bHasValidAimTarget!(self:Fort AIFauna Anim Instance Object Reference=@n1) -> (bHasValidAimTarget:bool)
   n3 = Get bIsBeingHeld!(self:Fort AIFauna Anim Instance Nug Object Reference=@n1) -> (bIsBeingHeld:bool)
   > Set ShouldPlayIdleBreak(ShouldPlayIdleBreak:bool=AND Boolean(A=NOT Boolean(A=@n2), B=NOT Boolean(A=@n3))) -> (Output_Get:bool)
   Event BlueprintInitializeAnimation()
   > n4 = Cast To AnimBP_Fauna_Nug!(execute, Object=Get MainAnimBP() -> (MainAnimBP:Anim Instance Object Reference)) -> (then, CastFailed, AsAnim BP Fauna Nug:Anim BP Fauna Nug Object Reference) {is=title K2Node_DynamicCast}
   > Set NugAnimBP(NugAnimBP:Anim BP Fauna Nug Object Reference=@n4) -> (Output_Get:Anim BP Fauna Nug Object Reference)
  ]}
 Fauna Idle {nodes [
   Input Pose() {name=InPose Idle}
   n1 = Random Sequence Player!() -> (Pose:Pose Link Structure) {
    is=title AnimGraphNode_RandomPlayer
    entries [
     {
      chanceToPlay=0.5
      maxLoopCount=0
      maxPlayRate=1
      minLoopCount=0
      minPlayRate=1
      sequence=/Irwin/AI/Prey/Nug/Animations/Nug_Idle
      blendIn {blendOption=Linear, blendTime=0.2, customCurve=null}
     }
     {
      chanceToPlay=0.5
      maxLoopCount=0
      maxPlayRate=1
      minLoopCount=0
      minPlayRate=1
      sequence=/Irwin/AI/Prey/Nug/Animations/Nug_IdleBreak_PeckingGround
      blendIn {blendOption=Linear, blendTime=0.2, customCurve=null}
     }
    ]
   }
   Output Pose(Result=Blend Poses by bool(
     BlendPose_0=@n1
     BlendPose_1=Sequence Player 'Nug_Idle'() {sequence=&Nug_Idle, startPosition=0.3}
     BlendTime_0=0.35
     BlendTime_1=0.35
     bActiveValue=Get ShouldPlayIdleBreak() -> (ShouldPlayIdleBreak:bool)
    ) {blendType=Linear, transitionType=Inertialization}) {layerGroup=FaunaIdleOverride}
  ]}
 Fauna IdleTurn {nodes [Output Pose(Result=Input Pose() {name=InPose IdleTurn}) {layerGroup=FaunaIdleOverride}]}
]
variables [
 NugAnimBP {type=AnimBP_Fauna_Nug_C, default=null}
 ShouldPlayIdleBreak {type=boolean}
]
